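# Physics asset for the Birthday Cake backpack cosmetic, giving its skeletal mesh a simple two-body physics setup.
# One capsule covers backpack_root and a smaller capsule covers the dyn_child_01 bone.
# A single constraint joins the child to the root, with the first swing axis locked and narrow angle limits set for the second swing axis and twist.
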PhysicsAsset /BRCosmetics/Accessories/FORT_Backpacks/M_MED_BirthdayCake_Backpack/Meshes/Backpack_BirthdayCake_PhysicsAsset
bodies {
 backpack_root [{shapeType=Capsule, length=22.9537, radius=14.9652, center=<7.044 -10.5816 -0.3531>, rotation=<rot 90 18.9704 13.9697>}]
 dyn_child_01 [{shapeType=Capsule, length=13.8806, radius=5.5489, center=<-0.2475 -0.0037 4.0472>, rotation=<rot 0 0 -90>}]
}
constraints [{bone1Name=dyn_child_01, bone2Name=backpack_root, swing1LimitDegrees=0, swing1Motion=Locked, swing2LimitDegrees=5, twistLimitDegrees=5}]
