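# Animation Blueprint for the Birthday Cake backpack cosmetic skeleton, adding simulated secondary motion to the backpack.
# The incoming Input Pose passes through a RigidBody node using the backpack's override physics asset, with scaled component velocity and acceleration, then goes to the output.
# An Anim Dynamics node on dyn_child_01 and a Copy Pose From Mesh node sit unconnected, and the event graph holds only empty default nodes.
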AnimBlueprint /BRCosmetics/Accessories/FORT_Backpacks/M_MED_BirthdayCake_Backpack/Meshes/Backpack_BirthdayCake_Skeleton_AnimBP
target_skeleton=&/Backpack_BirthdayCake_Skeleton
graphs [
 AnimGraph {nodes [
   Anim Dynamics - Bone: dyn_child_01(Alpha=1, ExternalForce="0,0,0") {
    angularSpringConstant=2000
    bAngularSpring=true
    bGravityOverrideInSimSpace=true
    alphaBoolBlend {blendOption=Linear}
    boundBone {boneName=dyn_child_01}
    componentAppliedLinearAccClamp=<1000 1000 1000>
    componentLinearAccScale=<1.2 1.2 1.2>
    componentLinearVelScale=<1 1 1>
    physicsBodyDefinitions [{
      sphereCollisionRadius=0
      boundBone {boneName=dyn_child_01}
      boxExtents=<40 40 40>
      constraintSetup {twistAxis=AxisZ, angularLimitsMax=<43.2 0 0>, angularLimitsMin=<0 0 0>, angularTarget=<0 0 0>}
      localJointOffset=<0 0 -2>
     }]
   }
   Copy Pose From Mesh() {bUseAttachedParent=true}
   AnimGraphNode_SubInput_0 = Input Pose()
   n1 = RigidBody(ComponentPose=Local To Component(LocalPose=@AnimGraphNode_SubInput_0)) {
    overridePhysicsAsset=&/Backpack_BirthdayCake_PhysicsAsset
    alphaBoolBlend {blendOption=Linear}
    componentAppliedLinearAccClamp=<1000 1000 1000>
    componentLinearAccScale=<1.1 1.1 1.1>
    componentLinearVelScale=<5 5 5>
   }
   Output Pose(Result=Component To Local(ComponentPose=@n1)) {layerGroup=None}
  ]}
 EventGraph {nodes [
   TryGetPawnOwner()
   Event BlueprintUpdateAnimation()
  ]}
]
variables []
